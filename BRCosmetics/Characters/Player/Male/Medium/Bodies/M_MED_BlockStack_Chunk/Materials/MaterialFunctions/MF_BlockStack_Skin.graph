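# Material function for BlockStack character skin that applies a tinted subsurface-profile look to masked areas of an input material.
# It multiplies BaseColor by SkinTint_BlockStack, sets the Subsurface Profile shading model and uses SkinSubsurfaceIntensity as Opacity.
# The result blends over the original attributes by the Skin Mask input.
# Curve-atlas skintone and older Fresnel skin-boost switch branches are present but do not feed the output.
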MaterialFunction /BRCosmetics/Characters/Player/Male/Medium/Bodies/M_MED_BlockStack_Chunk/Materials/MaterialFunctions/MF_BlockStack_Skin
graph {nodes [
  n1 = GetMaterialAttributes(None=FunctionInput() {inputName=Material, inputType=FunctionInput_MaterialAttributes}) {attributeGetTypes [
    #1
    #2
    #3
    #4
   ]}
  n2 = Multiply(A=@n1.BaseColor, B=VectorParameter() {group=02_Skin|BlockStack, parameterName=SkinTint_BlockStack, defaultValue=<rgba 1 1 1 1>}.RGB)
  n3 = ShadingModel() {shadingModel=MSM_SubsurfaceProfile}
  n4 = ScalarParameter() {defaultValue=0.4, group=02_Skin, parameterName=SkinSubsurfaceIntensity}
  FunctionOutput(None=BlendMaterialAttributes(
    A=@n1.MaterialAttributes
    B=SetMaterialAttributes!(MaterialAttributes=@n1.MaterialAttributes, BaseColor=@n2, ShadingModel=@n3, Opacity=@n4) {attributeSetTypes [#1, #5, #6]}
    Alpha=FunctionInput() {inputName=Skin Mask, inputType=FunctionInput_Scalar}
   ))
  Reroute(None=@n1.MaterialAttributes)
  n5 = CurveAtlasRowParameter(CurveTime=FunctionInput() {inputName=GMap, inputType=FunctionInput_Scalar}) {
   atlas=/Game/Global/Textures/Default/Curve/FN_Character_Skintone_Atlas
   curve=/Game/Global/Textures/Default/Curve/SkintoneGradients/Skintone_42
   defaultValue=41
   group=02_Skin|BlockStack
   parameterName=Skintone
   sortPriority=34
  }
  StaticSwitchParameter(True=@n5, False=@n1.BaseColor) {group=02_Skin|BlockStack, parameterName=Use Engine Colorized Skintones, sortPriority=33}
  n6 = VectorParameter() {group=02_Skin|BlockStack, parameterName=Skin Boost Color And Exponent, defaultValue=<rgba 0.87 0.6688 0.6948 2>}
  n7 = OneMinus(None=Fresnel(ExponentIn=@n6.A) {exponent=2})
  n8 = Multiply(A=@n1.BaseColor, B=@n6.RGB)
  n9 = Multiply(
   A=ScalarParameter() {defaultValue=0.75, group=02_Skin|BlockStack, parameterName=Global Skin Boost EX}
   B=Add(A=Multiply(A=@n7, B=@n8), B=@n8)
  )
  n10 = ScalarParameter() {defaultValue=1, group=02_Skin|BlockStack, parameterName=HumanSkin}
  StaticSwitchParameter(True=Saturate(None=LinearInterpolate(A=@n1.BaseColor, B=@n9, Alpha=@n10)), False=@n2) {group=02_Skin, parameterName=UseOldSkinShader}
 ]}
