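# Material function that supplies skin texture inputs for the BlockStack character body material, outputting material attributes plus a separate mask.
# It samples diffuse, SRM, normal and mask parameters on an input UV with shared wrap samplers.
# Although SRM is labeled spec/roughness/metal, its blue channel feeds Metallic, Specular and Roughness alike.
# Static switches for metal-mask power and disabling diffuse, SRM or mask samplers exist but feed no output.
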MaterialFunction /BRCosmetics/Characters/Player/Male/Medium/Bodies/M_MED_BlockStack_Chunk/Materials/MaterialFunctions/MF_BlockStack_Skin_Texture_Inputs
graph {nodes [
  n1 = FunctionInput(Preview=TextureCoordinate()) {bUsePreviewValueAsDefault=true, inputName=UV, inputType=FunctionInput_Vector2}
  n2 = TextureSampleParameter2D(UVs=@n1) {
   desc="R = Spec , G = Roughness, B= Metal"
   group=00_Texture Inputs|Skin
   parameterName=Skin_SRM
   samplerSource=SSM_Wrap_WorldGroupSettings
   samplerType=SAMPLERTYPE_Masks
   sortPriority=1
   texture=&T_Fortnite_Default_S
  }
  n3 = MaterialFunctionCall(Float3=@n2.RGB) {materialFunction=&BreakOutFloat3Components}
  n4 = Constant()
  n5 = TextureSampleParameter2D(UVs=@n1) {
   group=00_Texture Inputs|Skin
   parameterName=Skin_Normals
   samplerSource=SSM_Wrap_WorldGroupSettings
   samplerType=SAMPLERTYPE_Normal
   sortPriority=2
   texture=&FlatNormalBlue
  }
  n6 = TextureSampleParameter2D(UVs=@n1) {
   group=00_Texture Inputs|Skin
   parameterName=Skin_Diffuse
   samplerSource=SSM_Wrap_WorldGroupSettings
   sortPriority=0
   texture=/Game/Global/Textures/Default/Blanks/T_MidgreysRGB
  }
  FunctionOutput(None=MakeMaterialAttributes!(
    BaseColor=@n6.RGB
    Metallic=@n3.B
    Specular=@n3.B
    Roughness=@n3.B
    Anisotropy
    EmissiveColor=@n4
    Opacity
    OpacityMask
    Normal=@n5.RGB
    Tangent
    WorldPositionOffset
    SubsurfaceColor=@n4
    ClearCoat=@n4
    ClearCoatRoughness=@n4
    AmbientOcclusion
    Refraction
    CustomizedUVs_0
    CustomizedUVs_1
    CustomizedUVs_2
    CustomizedUVs_3
    CustomizedUVs_4
    CustomizedUVs_5
    CustomizedUVs_6
    CustomizedUVs_7
    PixelDepthOffset
    ShadingModel=ShadingModel()
    Displacement
   ))
  n7 = TextureSampleParameter2D(UVs=@n1) {
   group=00_Texture Inputs|Skin
   parameterName=Skin_Mask
   samplerSource=SSM_Wrap_WorldGroupSettings
   samplerType=SAMPLERTYPE_Masks
   sortPriority=3
   texture=&T_BlackMask
  }
  FunctionOutput(None=@n7.RGB) {outputName=M, sortPriority=4}
  n8 = ScalarParameter() {defaultValue=4, group=00_Texture Inputs, parameterName=MetalMaskPowerIntensity}
  StaticSwitchParameter(True=Saturate(None=Power(Base=@n3.B, Exp=@n8) {constExponent=4}), False=@n3.B) {group=00_Texture Inputs, parameterName=MetalMaskPower}
  StaticSwitchParameter(True=@n6.RGB, False=Constant()) {defaultValue=true, group=09_DisableTextureSamplers, parameterName=useDiffuseMap}
  StaticSwitchParameter(True=@n7.RGB, False=Constant()) {defaultValue=true, group=09_DisableTextureSamplers, parameterName=useMask}
  StaticSwitchParameter(True=@n2.RGB, False=Constant3Vector() {constant=<rgba 0.5 0 0.7 1>}) {defaultValue=true, group=09_DisableTextureSamplers, parameterName=useSRM}
 ]}
