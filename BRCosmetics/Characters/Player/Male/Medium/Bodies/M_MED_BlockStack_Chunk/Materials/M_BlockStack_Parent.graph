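# Parent material for BlockStack player character bodies, with switchable skin, clothing, color customization and a flipbook face.
# It wraps these layers between the shared 2023 character PreFX and PostFX functions.
# Customization uses channel masks and optional tiled, offset patterns to blend two tint colors and adjust metallic, roughness, normal flatness and subsurface color.
# Face alpha times the AnimBP-driven HasPoseChanged feeds temporal responsiveness.
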Material /BRCosmetics/Characters/Player/Male/Medium/Bodies/M_MED_BlockStack_Chunk/Materials/M_BlockStack_Parent
graph {
 expressions {
  MaterialExpressionAdd_0 {
   class=MaterialExpressionAdd
   inputs [
    A {from=MaterialExpressionReroute_2}
    B {from=MaterialExpressionScalarParameter_6}
   ]
  }
  MaterialExpressionAdd_1 {
   class=MaterialExpressionAdd
   inputs [
    A {from=MaterialExpressionReroute_4}
    B {from=MaterialExpressionScalarParameter_8}
   ]
  }
  MaterialExpressionAdd_2 {
   class=MaterialExpressionAdd
   inputs [
    A {from=MaterialExpressionTextureCoordinate_0}
    B {from=MaterialExpressionMultiply_6}
   ]
  }
  MaterialExpressionBlendMaterialAttributes_0 {
   class=MaterialExpressionBlendMaterialAttributes
   inputs [
    A {from=MaterialExpressionNamedRerouteUsage_20, output=""}
    B {from=MaterialExpressionNamedRerouteUsage_24, output=""}
    Alpha {from=MaterialExpressionNamedRerouteUsage_2, output=""}
   ]
   properties {}
  }
  MaterialExpressionBlendMaterialAttributes_1 {
   class=MaterialExpressionBlendMaterialAttributes
   inputs [
    A {from=MaterialExpressionNamedRerouteUsage_14, output=""}
    B {from=MaterialExpressionNamedRerouteUsage_23, output=""}
    Alpha {from=MaterialExpressionNamedRerouteUsage_8, output=""}
   ]
   properties {}
  }
  MaterialExpressionChannelMaskParameter_0 {
   class=MaterialExpressionChannelMaskParameter
   inputs [None {from=MaterialExpressionNamedRerouteUsage_7, output=""}]
   properties {group=BlockStack - Color Customization, parameterName=ColorCustomizationMaskChannel}
  }
  MaterialExpressionChannelMaskParameter_1 {
   class=MaterialExpressionChannelMaskParameter
   inputs [None {from=MaterialExpressionNamedRerouteUsage_9, output=""}]
   properties {group=BlockStack - Color Customization, maskChannel=Blue, parameterName=ColorCustomizationChannel, defaultValue=<rgba 0 0 1 0>}
  }
  MaterialExpressionChannelMaskParameter_2 {
   class=MaterialExpressionChannelMaskParameter
   inputs [None {from=MaterialExpressionNamedRerouteUsage_13, output=""}]
   properties {group=02_Skin|BlockStack, parameterName=FXMask_SkinMask}
  }
  MaterialExpressionChannelMaskParameter_3 {
   class=MaterialExpressionChannelMaskParameter
   inputs [None {from=MaterialExpressionNamedRerouteUsage_15, output=""}]
   properties {group=BlockStack - Color Customization, parameterName=PatternMaskChannel}
  }
  MaterialExpressionChannelMaskParameter_4 {
   class=MaterialExpressionChannelMaskParameter
   inputs [None {from=MaterialExpressionNamedRerouteUsage_1, output=""}]
   properties {group=BlockStack - Color Customization, maskChannel=Green, parameterName=PatternOffsetMaskChannel, defaultValue=<rgba 0 1 0 0>}
  }
  MaterialExpressionComponentMask_0 {class=MaterialExpressionComponentMask, inputs [None {from=MaterialExpressionMaterialFunctionCall_2, output=M}], properties {b=true}}
  MaterialExpressionComponentMask_1 {class=MaterialExpressionComponentMask, inputs [None {from=MaterialExpressionMaterialFunctionCall_2, output=M}], properties {g=true}}
  MaterialExpressionConvert_0 {
   class=MaterialExpressionConvert
   inputs [
    None {from=MaterialExpressionScalarParameter_3}
    None {from=MaterialExpressionScalarParameter_4}
   ]
   properties {
    convertInputs [
     {
      type=Scalar
      expressionInput {
       expression=&self:MaterialExpressionScalarParameter_3
       inputName=None
       mask=0
       maskA=0
       maskB=0
       maskG=0
       maskR=0
       outputIndex=0
      }
     }
     {
      type=Scalar
      expressionInput {
       expression=&self:MaterialExpressionScalarParameter_4
       inputName=None
       mask=0
       maskA=0
       maskB=0
       maskG=0
       maskR=0
       outputIndex=0
      }
     }
    ]
    convertOutputs [{type=Vector2}]
   }
  }
  MaterialExpressionGetMaterialAttributes_0 {
   class=MaterialExpressionGetMaterialAttributes
   inputs [None {from=MaterialExpressionNamedRerouteUsage_0, output=""}]
   properties {attributeGetTypes [
     #1
     #2
     #3
     #4
     #5
    ]}
  }
  MaterialExpressionLinearInterpolate_0 {
   class=MaterialExpressionLinearInterpolate
   inputs [
    A {from=MaterialExpressionVectorParameter_1, output=RGB}
    B {from=MaterialExpressionVectorParameter_0, output=RGB}
    Alpha {from=MaterialExpressionVectorParameter_0, output=RGB}
   ]
  }
  MaterialExpressionLinearInterpolate_1 {
   class=MaterialExpressionLinearInterpolate
   inputs [
    A {from=MaterialExpressionLinearInterpolate_0}
    B {from=MaterialExpressionLinearInterpolate_11}
    Alpha {from=MaterialExpressionNamedRerouteUsage_6}
   ]
  }
  MaterialExpressionLinearInterpolate_10 {
   class=MaterialExpressionLinearInterpolate
   inputs [
    A {from=MaterialExpressionReroute_5}
    B {from=MaterialExpressionMaterialFunctionCall_4, output=Result}
    Alpha {from=MaterialExpressionNamedRerouteUsage_11}
   ]
  }
  MaterialExpressionLinearInterpolate_11 {
   class=MaterialExpressionLinearInterpolate
   inputs [
    A {from=MaterialExpressionVectorParameter_7, output=RGB}
    B {from=MaterialExpressionVectorParameter_5, output=RGB}
    Alpha {from=MaterialExpressionVectorParameter_5, output=RGB}
   ]
  }
  MaterialExpressionLinearInterpolate_2 {
   class=MaterialExpressionLinearInterpolate
   inputs [
    A {from=""}
    B {from=""}
    Alpha {from=MaterialExpressionMultiply_1}
   ]
   properties {constA=0.4}
  }
  MaterialExpressionLinearInterpolate_3 {
   class=MaterialExpressionLinearInterpolate
   inputs [
    A {from=MaterialExpressionLinearInterpolate_8}
    B {from=MaterialExpressionOneMinus_0}
    Alpha {from=MaterialExpressionScalarParameter_5}
   ]
  }
  MaterialExpressionLinearInterpolate_4 {
   class=MaterialExpressionLinearInterpolate
   inputs [
    A {from=MaterialExpressionReroute_0}
    B {from=MaterialExpressionMultiply_4}
    Alpha {from=MaterialExpressionNamedRerouteUsage_12}
   ]
  }
  MaterialExpressionLinearInterpolate_6 {
   class=MaterialExpressionLinearInterpolate
   inputs [
    A {from=MaterialExpressionMaterialFunctionCall_8, output=B}
    B {from=MaterialExpressionChannelMaskParameter_2, output=R}
    Alpha {from=MaterialExpressionScalarParameter_7}
   ]
  }
  MaterialExpressionLinearInterpolate_7 {
   class=MaterialExpressionLinearInterpolate
   inputs [
    A {from=MaterialExpressionReroute_2}
    B {from=MaterialExpressionSaturate_0}
    Alpha {from=MaterialExpressionNamedRerouteUsage_4}
   ]
  }
  MaterialExpressionLinearInterpolate_8 {
   class=MaterialExpressionLinearInterpolate
   inputs [
    A {from=""}
    B {from=MaterialExpressionChannelMaskParameter_0, output=R}
    Alpha {from=MaterialExpressionScalarParameter_0}
   ]
   properties {constA=1}
  }
  MaterialExpressionLinearInterpolate_9 {
   class=MaterialExpressionLinearInterpolate
   inputs [
    A {from=MaterialExpressionReroute_4}
    B {from=MaterialExpressionSaturate_1}
    Alpha {from=MaterialExpressionNamedRerouteUsage_10}
   ]
  }
  MaterialExpressionMaterialFunctionCall_0 {
   class=MaterialExpressionMaterialFunctionCall
   inputs [
    Overide Texture Set UVs {from=""}
    Override Tech Art Mask UVs {from=""}
    Override Basic Emissive UVs {from=""}
   ]
   properties {materialFunction=&MF_CharacterParent_2023_PreFX}
  }
  MaterialExpressionMaterialFunctionCall_1 {
   class=MaterialExpressionMaterialFunctionCall
   inputs [
    Material {from=MaterialExpressionNamedRerouteUsage_26}
    M {from=MaterialExpressionMaterialFunctionCall_0, output=M}
    Pre Skinned Local Position {from=MaterialExpressionMaterialFunctionCall_0, output=M}
    Distance Blend {from=MaterialExpressionMaterialFunctionCall_0, output=M}
    Fresnel {from=MaterialExpressionMaterialFunctionCall_0, output=M}
   ]
   properties {materialFunction=&MF_CharacterParent_2023_PostFX}
  }
  MaterialExpressionMaterialFunctionCall_2 {
   class=MaterialExpressionMaterialFunctionCall
   inputs [UV {from=""}]
   properties {materialFunction=/BRCosmetics/Characters/Player/Male/Medium/Bodies/M_MED_BlockStack_Chunk/Materials/MaterialFunctions/MF_BlockStack_Skin_Texture_Inputs}
  }
  MaterialExpressionMaterialFunctionCall_3 {
   class=MaterialExpressionMaterialFunctionCall
   inputs [
    Skin Mask {from=MaterialExpressionComponentMask_0}
    GMap {from=MaterialExpressionReroute_1}
    Material {from=MaterialExpressionMaterialFunctionCall_2, output=Result}
   ]
   properties {materialFunction=/BRCosmetics/Characters/Player/Male/Medium/Bodies/M_MED_BlockStack_Chunk/Materials/MaterialFunctions/MF_BlockStack_Skin}
  }
  MaterialExpressionMaterialFunctionCall_4 {
   class=MaterialExpressionMaterialFunctionCall
   inputs [
    Normal {from=MaterialExpressionReroute_5}
    Flatness {from=MaterialExpressionScalarParameter_9}
   ]
   properties {materialFunction=&FlattenNormal}
  }
  MaterialExpressionMaterialFunctionCall_5 {
   class=MaterialExpressionMaterialFunctionCall
   inputs [
    CustomUVs {from=""}
    Material {from=MaterialExpressionStaticSwitchParameter_2}
   ]
   properties {materialFunction=/BRCosmetics/Characters/Player/Male/Medium/Bodies/M_MED_BlockStack_Chunk/Materials/MaterialFunctions/MF_BlockStack_FlipbookFace}
  }
  MaterialExpressionMaterialFunctionCall_8 {
   class=MaterialExpressionMaterialFunctionCall
   inputs [Float3 {from=MaterialExpressionNamedRerouteUsage_22}]
   properties {desc="R = AO\r\nG = GMap\r\nB = Skin", materialFunction=&BreakOutFloat3Components}
  }
  MaterialExpressionMultiply_0 {
   class=MaterialExpressionMultiply
   inputs [
    A {from=MaterialExpressionNamedRerouteUsage_18}
    B {from=MaterialExpressionGetMaterialAttributes_0, output=Base Color}
   ]
  }
  MaterialExpressionMultiply_1 {
   class=MaterialExpressionMultiply
   inputs [
    A {from=MaterialExpressionReroute_7}
    B {from=MaterialExpressionScalarParameter_2}
   ]
  }
  MaterialExpressionMultiply_2 {
   class=MaterialExpressionMultiply
   inputs [
    A {from=MaterialExpressionAdd_2}
    B {from=MaterialExpressionConvert_0}
   ]
  }
  MaterialExpressionMultiply_3 {
   class=MaterialExpressionMultiply
   inputs [
    A {from=MaterialExpressionReroute_6}
    B {from=MaterialExpressionLinearInterpolate_4}
   ]
  }
  MaterialExpressionMultiply_4 {
   class=MaterialExpressionMultiply
   inputs [
    A {from=MaterialExpressionReroute_0}
    B {from=MaterialExpressionScalarParameter_10}
   ]
   properties {constB=0.5}
  }
  MaterialExpressionMultiply_5 {
   class=MaterialExpressionMultiply
   inputs [
    A {from=MaterialExpressionTextureSampleParameter2D_0, output=RGB}
    B {from=MaterialExpressionChannelMaskParameter_3, output=R}
   ]
  }
  MaterialExpressionMultiply_6 {
   class=MaterialExpressionMultiply
   inputs [
    A {from=MaterialExpressionNamedRerouteUsage_16}
    B {from=MaterialExpressionScalarParameter_11}
   ]
   properties {constB=0}
  }
  MaterialExpressionNamedRerouteDeclaration_0 {
   class=MaterialExpressionNamedRerouteDeclaration
   inputs [None {from=MaterialExpressionMaterialFunctionCall_8, output=R}]
   properties {name=Ambient Occlusion, nodeColor=<rgba 0.875 0 1 1>}
  }
  MaterialExpressionNamedRerouteDeclaration_1 {
   class=MaterialExpressionNamedRerouteDeclaration
   inputs [None {from=MaterialExpressionMaterialFunctionCall_8, output=G}]
   properties {name=GMap, nodeColor=<rgba 0.6641 0 1 1>}
  }
  MaterialExpressionNamedRerouteDeclaration_11 {
   class=MaterialExpressionNamedRerouteDeclaration
   inputs [None {from=MaterialExpressionLinearInterpolate_1}]
   properties {name=Input Colors, nodeColor=<rgba 1 0 0.0234 1>}
  }
  MaterialExpressionNamedRerouteDeclaration_12 {
   class=MaterialExpressionNamedRerouteDeclaration
   inputs [None {from=MaterialExpressionStaticSwitchParameter_0}]
   properties {name=Skin, nodeColor=<rgba 0 0.3438 1 1>}
  }
  MaterialExpressionNamedRerouteDeclaration_13 {
   class=MaterialExpressionNamedRerouteDeclaration
   inputs [None {from=MaterialExpressionMaterialFunctionCall_0, output=M}]
   properties {name=Mask In, nodeColor=<rgba 0.6641 1 0 1>}
  }
  MaterialExpressionNamedRerouteDeclaration_14 {
   class=MaterialExpressionNamedRerouteDeclaration
   inputs [None {from=MaterialExpressionSetMaterialAttributes_0}]
   properties {name=Patterns, nodeColor=<rgba 1 0 0.9844 1>}
  }
  MaterialExpressionNamedRerouteDeclaration_15 {
   class=MaterialExpressionNamedRerouteDeclaration
   inputs [None {from=MaterialExpressionMaterialFunctionCall_5, output=Material}]
   properties {name=Material Out, nodeColor=<rgba 0 0.3964 1 1>}
  }
  MaterialExpressionNamedRerouteDeclaration_2 {
   class=MaterialExpressionNamedRerouteDeclaration
   inputs [None {from=MaterialExpressionLinearInterpolate_6}]
   properties {name=Skin Mask, nodeColor=<rgba 0.6641 0 1 1>}
  }
  MaterialExpressionNamedRerouteDeclaration_3 {
   class=MaterialExpressionNamedRerouteDeclaration
   inputs [None {from=MaterialExpressionMaterialFunctionCall_0, output=Result}]
   properties {name=Material In, nodeColor=<rgba 0 1 0.7656 1>}
  }
  MaterialExpressionNamedRerouteDeclaration_4 {
   class=MaterialExpressionNamedRerouteDeclaration
   inputs [None {from=MaterialExpressionMaterialFunctionCall_0, output=Tech Art Mask}]
   properties {name=TechArtNMask In, nodeColor=<rgba 1 0.4453 0 1>}
  }
  MaterialExpressionNamedRerouteDeclaration_5 {
   class=MaterialExpressionNamedRerouteDeclaration
   inputs [None {from=MaterialExpressionStaticSwitchParameter_3}]
   properties {name=PatternMask, nodeColor=<rgba 0.5234 0 1 1>}
  }
  MaterialExpressionNamedRerouteDeclaration_6 {
   class=MaterialExpressionNamedRerouteDeclaration
   inputs [None {from=MaterialExpressionStaticSwitchParameter_1}]
   properties {name=ColorCustomization, nodeColor=<rgba 1 0.9609 0 1>}
  }
  MaterialExpressionNamedRerouteDeclaration_7 {
   class=MaterialExpressionNamedRerouteDeclaration
   inputs [None {from=MaterialExpressionLinearInterpolate_3}]
   properties {name=ColorCustomizationMask, nodeColor=<rgba 1 0.0605 0 1>}
  }
  MaterialExpressionNamedRerouteDeclaration_8 {
   class=MaterialExpressionNamedRerouteDeclaration
   inputs [None {from=MaterialExpressionChannelMaskParameter_4, output=R}]
   properties {name=PatternOffsetMask, nodeColor=<rgba 1 0 0.6562 1>}
  }
  MaterialExpressionNamedRerouteDeclaration_9 {
   class=MaterialExpressionNamedRerouteDeclaration
   inputs [None {from=MaterialExpressionMultiply_2}]
   properties {name=Pattern UVs, nodeColor=<rgba 0 0.9766 1 1>}
  }
  MaterialExpressionNamedRerouteUsage_0 {class=MaterialExpressionNamedRerouteUsage}
  MaterialExpressionNamedRerouteUsage_1 {class=MaterialExpressionNamedRerouteUsage}
  MaterialExpressionNamedRerouteUsage_10 {class=MaterialExpressionNamedRerouteUsage}
  MaterialExpressionNamedRerouteUsage_11 {class=MaterialExpressionNamedRerouteUsage}
  MaterialExpressionNamedRerouteUsage_12 {class=MaterialExpressionNamedRerouteUsage}
  MaterialExpressionNamedRerouteUsage_13 {class=MaterialExpressionNamedRerouteUsage}
  MaterialExpressionNamedRerouteUsage_14 {class=MaterialExpressionNamedRerouteUsage}
  MaterialExpressionNamedRerouteUsage_15 {class=MaterialExpressionNamedRerouteUsage}
  MaterialExpressionNamedRerouteUsage_16 {class=MaterialExpressionNamedRerouteUsage}
  MaterialExpressionNamedRerouteUsage_17 {class=MaterialExpressionNamedRerouteUsage}
  MaterialExpressionNamedRerouteUsage_18 {class=MaterialExpressionNamedRerouteUsage}
  MaterialExpressionNamedRerouteUsage_19 {class=MaterialExpressionNamedRerouteUsage}
  MaterialExpressionNamedRerouteUsage_2 {class=MaterialExpressionNamedRerouteUsage}
  MaterialExpressionNamedRerouteUsage_20 {class=MaterialExpressionNamedRerouteUsage}
  MaterialExpressionNamedRerouteUsage_22 {class=MaterialExpressionNamedRerouteUsage}
  MaterialExpressionNamedRerouteUsage_23 {class=MaterialExpressionNamedRerouteUsage}
  MaterialExpressionNamedRerouteUsage_24 {class=MaterialExpressionNamedRerouteUsage}
  MaterialExpressionNamedRerouteUsage_25 {class=MaterialExpressionNamedRerouteUsage}
  MaterialExpressionNamedRerouteUsage_26 {class=MaterialExpressionNamedRerouteUsage}
  MaterialExpressionNamedRerouteUsage_3 {class=MaterialExpressionNamedRerouteUsage}
  MaterialExpressionNamedRerouteUsage_4 {class=MaterialExpressionNamedRerouteUsage}
  MaterialExpressionNamedRerouteUsage_5 {class=MaterialExpressionNamedRerouteUsage}
  MaterialExpressionNamedRerouteUsage_6 {class=MaterialExpressionNamedRerouteUsage}
  MaterialExpressionNamedRerouteUsage_7 {class=MaterialExpressionNamedRerouteUsage}
  MaterialExpressionNamedRerouteUsage_8 {class=MaterialExpressionNamedRerouteUsage}
  MaterialExpressionNamedRerouteUsage_9 {class=MaterialExpressionNamedRerouteUsage}
  MaterialExpressionOneMinus_0 {class=MaterialExpressionOneMinus, inputs [None {from=MaterialExpressionLinearInterpolate_8}]}
  MaterialExpressionReroute_0 {class=MaterialExpressionReroute, inputs [None {from=MaterialExpressionNamedRerouteUsage_19}]}
  MaterialExpressionReroute_1 {class=MaterialExpressionReroute, inputs [None {from=MaterialExpressionComponentMask_1}]}
  MaterialExpressionReroute_2 {class=MaterialExpressionReroute, inputs [None {from=MaterialExpressionGetMaterialAttributes_0, output=Metallic}]}
  MaterialExpressionReroute_4 {class=MaterialExpressionReroute, inputs [None {from=MaterialExpressionGetMaterialAttributes_0, output=Roughness}]}
  MaterialExpressionReroute_5 {class=MaterialExpressionReroute, inputs [None {from=MaterialExpressionGetMaterialAttributes_0, output=Normal}]}
  MaterialExpressionReroute_6 {class=MaterialExpressionReroute, inputs [None {from=MaterialExpressionGetMaterialAttributes_0, output=Fuzz Color or Subsurface Color}]}
  MaterialExpressionReroute_7 {class=MaterialExpressionReroute, inputs [None {from=MaterialExpressionMaterialFunctionCall_5, output=FaceAlpha}]}
  MaterialExpressionSaturate_0 {class=MaterialExpressionSaturate, inputs [None {from=MaterialExpressionAdd_0}]}
  MaterialExpressionSaturate_1 {class=MaterialExpressionSaturate, inputs [None {from=MaterialExpressionAdd_1}]}
  MaterialExpressionScalarParameter_0 {
   class=MaterialExpressionScalarParameter
   properties {
    controlType=Enumeration
    enumeration=/Game/Global/Materials/Enums/E_Mat_OnOff
    group=BlockStack - Color Customization
    parameterName=UseTechArtMaskForColorCustomization
   }
  }
  MaterialExpressionScalarParameter_1 {class=MaterialExpressionScalarParameter, properties {group=BlockStack - Color Customization, parameterName=CustomizationColorIndex}}
  MaterialExpressionScalarParameter_10 {
   class=MaterialExpressionScalarParameter
   properties {defaultValue=1, group=BlockStack - Color Customization, parameterName=SecondarySSSColorMultiplier}
  }
  MaterialExpressionScalarParameter_11 {class=MaterialExpressionScalarParameter, properties {group=BlockStack - Color Customization, parameterName=PatternOffsetAmount}}
  MaterialExpressionScalarParameter_2 {
   class=MaterialExpressionScalarParameter
   properties {desc="Set by base head AnimBP if \"IsFlipbookFace\" is true", group=Flipbook 01 Global, parameterName=HasPoseChanged}
  }
  MaterialExpressionScalarParameter_3 {
   class=MaterialExpressionScalarParameter
   properties {defaultValue=8, group=BlockStack - Color Customization, parameterName=CustomizationMask_Tiling_X}
  }
  MaterialExpressionScalarParameter_4 {
   class=MaterialExpressionScalarParameter
   properties {defaultValue=8, group=BlockStack - Color Customization, parameterName=CustomizationMask_Tiling_Y}
  }
  MaterialExpressionScalarParameter_5 {
   class=MaterialExpressionScalarParameter
   properties {
    controlType=Enumeration
    enumeration=&E_Mat_OnOff
    group=BlockStack - Color Customization
    parameterName=InvertCustomizationMask
   }
  }
  MaterialExpressionScalarParameter_6 {class=MaterialExpressionScalarParameter, properties {group=BlockStack - Color Customization, parameterName=MetallicBias}}
  MaterialExpressionScalarParameter_7 {class=MaterialExpressionScalarParameter, properties {group=02_Skin|BlockStack, parameterName=UseFXMaskAsSkinMask}}
  MaterialExpressionScalarParameter_8 {class=MaterialExpressionScalarParameter, properties {group=BlockStack - Color Customization, parameterName=RoughnessBias}}
  MaterialExpressionScalarParameter_9 {class=MaterialExpressionScalarParameter, properties {group=BlockStack - Color Customization, parameterName=NormalFlatness}}
  MaterialExpressionSetMaterialAttributes_0 {
   class=MaterialExpressionSetMaterialAttributes
   inputs [
    MaterialAttributes {from=MaterialExpressionGetMaterialAttributes_0, output=MaterialAttributes}
    Base Color {from=MaterialExpressionMultiply_0, output=""}
    Metallic {from=MaterialExpressionLinearInterpolate_7, output=""}
    Roughness {from=MaterialExpressionLinearInterpolate_9, output=""}
    Normal {from=MaterialExpressionLinearInterpolate_10, output=""}
    Fuzz Color or Subsurface Color {from=MaterialExpressionMultiply_3, output=""}
   ]
   properties {attributeSetTypes [
     #1
     #2
     #3
     #4
     #5
    ]}
  }
  MaterialExpressionStaticSwitchParameter_0 {
   class=MaterialExpressionStaticSwitchParameter
   inputs [
    True {from=MaterialExpressionMaterialFunctionCall_3, output=Result}
    False {from=MaterialExpressionNamedRerouteUsage_3, output=""}
   ]
   properties {defaultValue=true, group=02_Skin|BlockStack, parameterName=UseBlockStackSkin}
  }
  MaterialExpressionStaticSwitchParameter_1 {
   class=MaterialExpressionStaticSwitchParameter
   inputs [
    True {from=MaterialExpressionBlendMaterialAttributes_1, output=""}
    False {from=MaterialExpressionNamedRerouteUsage_5, output=""}
   ]
   properties {group=BlockStack - Color Customization, parameterName=UseColorCustomization}
  }
  MaterialExpressionStaticSwitchParameter_2 {
   class=MaterialExpressionStaticSwitchParameter
   inputs [
    True {from=MaterialExpressionBlendMaterialAttributes_0, output=""}
    False {from=MaterialExpressionNamedRerouteUsage_25, output=""}
   ]
   properties {defaultValue=true, group=02_Skin|BlockStack, parameterName=UseBlockStackClothing}
  }
  MaterialExpressionStaticSwitchParameter_3 {
   class=MaterialExpressionStaticSwitchParameter
   inputs [
    True {from=MaterialExpressionMultiply_5, output=""}
    False {from=MaterialExpressionStaticSwitchParameter_5, output=""}
   ]
   properties {group=BlockStack - Color Customization, parameterName=UsePatternTexture}
  }
  MaterialExpressionStaticSwitchParameter_5 {
   class=MaterialExpressionStaticSwitchParameter
   inputs [
    True {from=MaterialExpressionChannelMaskParameter_1, output=R}
    False {from=MaterialExpressionScalarParameter_1, output=""}
   ]
   properties {group=BlockStack - Color Customization, parameterName=UseFXMaskAsPatternTexture}
  }
  MaterialExpressionTemporalResponsivenessOutput_0 {class=MaterialExpressionTemporalResponsivenessOutput, inputs [None {from=MaterialExpressionLinearInterpolate_2, output=""}], properties {}}
  MaterialExpressionTextureCoordinate_0 {class=MaterialExpressionTextureCoordinate}
  MaterialExpressionTextureSampleParameter2D_0 {
   class=MaterialExpressionTextureSampleParameter2D
   inputs [
    UVs {from=MaterialExpressionNamedRerouteUsage_17, output=""}
    Apply View MipBias {from="", output=""}
   ]
   properties {group=BlockStack - Color Customization, parameterName=ColorCustomizationMask, texture=&Black}
  }
  MaterialExpressionVectorParameter_0 {
   class=MaterialExpressionVectorParameter
   properties {
    desc="if A=0.0 ColorLitDefault is used"
    group=BlockStack - Color Customization
    parameterName=ColorLit_R_Mask
    sortPriority=15
    defaultValue=<rgba 1 0 0 0>
   }
  }
  MaterialExpressionVectorParameter_1 {
   class=MaterialExpressionVectorParameter
   properties {group=BlockStack - Color Customization, parameterName=ColorLitDefault_R_Mask, sortPriority=23, defaultValue=<rgba 1 0 0 0>}
  }
  MaterialExpressionVectorParameter_5 {
   class=MaterialExpressionVectorParameter
   properties {
    desc="if A=0.0 ColorLitDefault is used"
    group=BlockStack - Color Customization
    parameterName=ColorLit_G_Mask
    sortPriority=16
    defaultValue=<rgba 0 1 0 0>
   }
  }
  MaterialExpressionVectorParameter_7 {
   class=MaterialExpressionVectorParameter
   properties {group=BlockStack - Color Customization, parameterName=ColorLitDefault_G_Mask, sortPriority=24, defaultValue=<rgba 0 1 0 0>}
  }
 }
 properties {
  bAutomaticallySetUsageInEditor=false
  bUseMaterialAttributes=true
  bUsedWithMeshDeformer=true
  bUsedWithMorphTargets=true
  bUsedWithSkeletalMesh=true
  pixelDepthOffsetMode=PDOM_Legacy
  previewMesh=/BRCosmetics/Gadgets/Assets/VinderTech_GliderChute/Glider_BlockStack/Meshes/BlockStack_Glider
  shadingModel=MSM_FromMaterialExpression
  usedShadingModels=Default Lit | Subsurface | Clear Coat | Subsurface Profile | Cloth
 }
}
