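# Physics asset of colliders for the medium male BlockStack Chunk player cosmetic body.
# One capsule per bone covers pelvis, spine, back of chest, thighs, calves, feet, and upper and lower arms, with no head or hand bodies.
# Only two constraints exist, joining each foot to its calf with free swing and twist, so the asset does not form a fully jointed ragdoll.
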PhysicsAsset /BRCosmetics/Characters/Player/Male/Medium/Bodies/M_MED_BlockStack_Chunk/Meshes/M_MED_BlockStack_Chunk_Physics_Colliders
bodies {
 calf_l [{shapeType=Capsule, length=43.2566, radius=5.4832, center=<-25.3016 -0.2281 0.5721>, rotation=<rot -0.0013 89.0727 90.0079>}]
 calf_r [{shapeType=Capsule, length=38.0662, radius=6.706, center=<25.1693 -0.3502 -0.5721>, rotation=<rot -0.0017 91.6007 -89.9922>}]
 chest_back [{
   shapeType=Capsule
   length=22.0234
   radius=9.6255
   shapeName=chest_back_capsule
   center=<0.0348 7.1912 -3.7227>
   rotation=<rot 0 -91.4863 -89.1993>
  }]
 deform_foot_l [{shapeType=Capsule, length=13.2767, radius=10.1859, shapeName=deform_foot_l_capsule, center=<-2.0269 6.1856 0>, rotation=<rot 90 -90 0>}]
 deform_foot_r [{shapeType=Capsule, length=13.2767, radius=10.1859, shapeName=deform_foot_r_capsule, center=<2.0269 -6.1856 0>, rotation=<rot 90 90 0>}]
 deform_pelvis [{
   shapeType=Capsule
   length=12.3187
   radius=22.4814
   shapeName=deform_pelvis_capsule
   center=<4.826 3.4727 0.2866>
   rotation=<rot 1.1665 -89.9007 1.2006>
  }]
 deform_spine_05 [{
   shapeType=Capsule
   length=6.6149
   radius=23.3013
   shapeName=deform_spine_05_capsule
   center=<4.7576 8.6722 -0.0994>
   rotation=<rot 0.4982 -90 0.2747>
  }]
 deform_thigh_l [{
   shapeType=Capsule
   length=31.0675
   radius=6.1421
   shapeName=deform_thigh_l_capsule
   center=<-16.7423 0.1697 -0.0503>
   rotation=<rot 0 -90.3422 -88.2977>
  }]
 deform_thigh_r [{
   shapeType=Capsule
   length=31.0675
   radius=6.1421
   shapeName=deform_thigh_r_capsule
   center=<15.414 -0.1622 0.0899>
   rotation=<rot 0 -90.3422 -88.2976>
  }]
 lowerRetArm_l [{
   shapeType=Capsule
   length=31.9629
   radius=4.4153
   shapeName=lowerRetArm_l_capsule
   center=<14.9308 0.5221 -0.3276>
   rotation=<rot 0 87.2391 -89.5728>
  }]
 lowerRetArm_r [{
   shapeType=Capsule
   length=35.6875
   radius=4.4535
   shapeName=lowerRetArm_r_capsule
   center=<-18.3872 0.1352 0.2015>
   rotation=<rot 0 86.5683 -89.314>
  }]
 spine_02 [{shapeType=Capsule, length=19.0739, radius=17.2209, shapeName=spine_02_capsule, center=<10.3405 0.2527 0>, rotation=<rot 0 -90 0>}]
 upperRetArm_l [{
   shapeType=Capsule
   length=19.8974
   radius=5.1328
   shapeName=upperRetArm_l_capsule
   center=<12.6225 0.6062 -0.5079>
   rotation=<rot 0.0837 -86.4225 -88.6343>
  }]
 upperRetArm_r [{
   shapeType=Capsule
   length=23.868
   radius=5.1328
   shapeName=upperRetArm_r_capsule
   center=<-12.6224 -0.6062 0.5078>
   rotation=<rot 0.1304 -88.3858 -88.638>
  }]
}
constraints [
 {
  bone1Name=deform_foot_r
  bone2Name=calf_r
  swing1LimitDegrees=45
  swing1Motion=Free
  swing2LimitDegrees=45
  swing2Motion=Free
  twistLimitDegrees=45
  twistMotion=Free
 }
 {
  bone1Name=deform_foot_l
  bone2Name=calf_l
  swing1LimitDegrees=45
  swing1Motion=Free
  swing2LimitDegrees=45
  swing2Motion=Free
  twistLimitDegrees=45
  twistMotion=Free
 }
]
